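# Translucent deferred decal material that draws left and right footprints from Niagara particles for snow and sand interaction.
# UVs come from world position relative to the particle, unrotated by its quaternion, then mirrored and rotated per print.
# Blurred, sharp and optional detail normals combine with landscape runtime virtual texture normals.
# Opacity uses a mask texture, particle alpha and a CustomDepth cutoff, and is zero on Forward and Mobile paths.
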Material /Game/Effects/SnowSandInteraction/Materials/FootprintDecal/SSI_M_FootprintDecal
graph {
 properties {
  bUsedWithNiagaraMeshParticles=true
  bUsedWithNiagaraSprites=true
  blendMode=BLEND_Translucent
  materialDomain=MD_DeferredDecal
  pixelDepthOffsetMode=PDOM_Legacy
  usedShadingModels=Default Lit
 }
 nodes [
  n1 = DynamicParameter() {defaultValue=<rgba 1 1 0 0>, paramNames [CompressedQuaternion, BlurMarkScale, MarkFlat, MeshScale]}
  n2 = MaterialFunctionCall() {materialFunction=&WorldPositionBehindTranslucency}
  n3 = Subtract(A=@n2.Result, B=LinearInterpolate(A=ParticlePositionWS(), B=ActorPositionWS(), Alpha=ScalarParameter() {parameterName=UseDebugPosition}))
  n4 = DynamicParameter() {defaultValue=<rgba 1 1 0 0>, paramNames [CompressedQuaternion, BlurMarkScale, MarkFlat, MeshScale]}
  n5 = Multiply(A=@n4.MeshScale) {constB=100}
  n6 = DynamicParameter() {parameterIndex=3, defaultValue=<rgba 1 1 0 0>, paramNames [QuatX,QuatY,QuatZ,QuatW]}
  n7 = MaterialFunctionCall(R=@n6.QuatX, G=@n6.QuatX, B=@n6.QuatX, A=@n6.QuatX) {materialFunction=&AppendMany}
  n8 = StaticBool() {value=true}
  n9 = Add(A=MaterialFunctionCall!(Vector=Divide(A=@n3, B=@n5) {constB=100}, Quat=@n7.RGBA, Invert Quaternion=@n8) {materialFunction=/Engine/Functions/Engine_MaterialFunctions03/Math/MultiplyVectorWithQuaternion}.Result) {constB=0.5}
  n10 = LinearInterpolate(
   A=TextureCoordinate()
   B=Add(A=ComponentMask(None=@n9) {g=true, r=true}, B=Constant2Vector() {g=-0.05})
   Alpha=ScalarParameter() {defaultValue=1, parameterName=UseGroundNormalForTexCoord}
  )
  n11 = DynamicParameter() {parameterIndex=1, defaultValue=<rgba 1 1 0 0>, paramNames [Rotation, IsRight, Random, ""]}
  n12 = LinearInterpolate(A=Constant2Vector() {g=1, r=1}, B=Constant2Vector() {g=1, r=-1}, Alpha=@n11.IsRight)
  n13 = MaterialFunctionCall(
   UVs=MaterialFunctionCall(
    UVs=MaterialFunctionCall(UVs=@n10, Texture Scale=@n12) {materialFunction=&ScaleUVsByCenter}.UVs
    Rotation Center=Constant2Vector() {g=0.5, r=0.5}
    "Rotation Angle (0-1)"=@n11.Rotation
   ) {materialFunction=&CustomRotator}.Rotated Values
   Texture Scale=ScalarParameter() {defaultValue=1, parameterName=POM_UVScale}
  ) {materialFunction=&ScaleUVsByCenter}
  n14 = LinearInterpolate(
   B=@n13.UVs
   Alpha=MaterialFunctionCall!(TexCoord=@n13.UVs, ClampChannel=Constant() {r=1}) {materialFunction=/Game/Effects/Fort_Effects/MaterialFunctions/MF_FX_TilingMethod}.Result
  )
  n15 = MaterialFunctionCall(
   UVs=@n14
   Texture Scale=LinearInterpolate(
    A=ScalarParameter() {defaultValue=1, parameterName=ShoeMark_BlurMark_MinScale}
    B=ScalarParameter() {defaultValue=1, parameterName=ShoeMark_BlurMark_MaxScale}
    Alpha=@n1.BlurMarkScale
   ) {constA=1}
  ) {materialFunction=&ScaleUVsByCenter}
  MaterialExpressionTextureSampleParameter2D_3 = TextureSampleParameter2D(UVs=@n15.UVs) {parameterName=MaskTexture, texture=/Game/Effects/SnowSandInteraction/Textures/SSI_T_FootprintMasks}
  n16 = Custom!(Opacity=@MaterialExpressionTextureSampleParameter2D_3.G) {
   code="float mask = 0;\r\nif (Opacity >= 0.9)\r\n{\r\nclip(-1);\r\n}\r\nreturn mask;"
   outputType=CMOT_Float1
   inputs [{
     inputName=Opacity
     input {
      expression=&self:MaterialExpressionTextureSampleParameter2D_3
      mask=1
      maskG=1
      outputIndex=2
     }
    }]
  }
  n17 = Subtract(A=SceneTexture(UVs=ScreenPosition().ViewportUV) {sceneTextureId=PPI_CustomDepth}.Color) {constB=4096}
  n18 = MaterialFunctionCall(Float3=Saturate(None=Floor(None=@n17))) {materialFunction=&BreakOutFloat3Components}
  n19 = OneMinus(None=@MaterialExpressionTextureSampleParameter2D_3.G)
  n20 = Saturate(None=Multiply(A=ParticleColor().A, B=Multiply(A=@n18.R, B=@n19)))
  n21 = Constant()
  n22 = ShadingPathSwitch(Default=@n20, Deferred=@n20, Forward=@n21, Mobile=@n21)
  Add(
   A=@n16
   B=Multiply(A=TextureSample() {texture=/Game/Effects/Fort_Effects/Gameplay/Tutorial_Target/shader_mask_round_square_blurry}.RGB, B=@n22)
  )
  Add(
   A=@n22
   B=MaterialFunctionCall!(Vector3=NamedRerouteUsage(), MaximumNumberOfDigits, UVs, DebugTextLocation RG_UpperRight BA_LowerLeft, Component Spacing) {materialFunction=/Engine/Functions/Engine_MaterialFunctions02/Utility/DebugFloat3Values}.GreyScaleOutput
  )
  n23 = Transform(None=Constant3Vector() {constant=<rgba 0 0 1 0>})
  n24 = RuntimeVirtualTextureSampleParameter(World Position=@n2.Result) {
   group=RVT
   materialType=BaseColor_Normal_Specular_YCoCg
   parameterName=RVT Sample Target
   virtualTexture=/Game/Athena/Apollo/Maps/Landscape/VT/RVT_Apollo_Landscape
  }
  n25 = DotProduct(A=@n23, B=@n24.Normal)
  n26 = Power(Base=Saturate(None=Abs(None=@n25))) {constExponent=10}
  LinearInterpolate(A=Constant3Vector() {constant=<rgba 1 0 0 0>}, B=Constant3Vector() {constant=<rgba 0 1 0 0>}, Alpha=@n26)
  MaterialFunctionCall!(CompressedQuaternion=@n4.CompressedQuaternion) {materialFunction=/Niagara/Functions/DataPacking/Niagara_DecompressQuaternion}
  NamedRerouteDeclaration(None=LinearInterpolate(A=@n23, B=@n24.Normal, Alpha=@n26)) {name=RVTNormal, nodeColor=<rgba 0 1 0.9766 1>}
  n27 = TextureSampleParameter2D(UVs=@n15.UVs) {
   parameterName=ShoeMark_BlurMark
   samplerSource=SSM_Clamp_WorldGroupSettings
   samplerType=SAMPLERTYPE_Normal
   texture=/Game/Effects/SnowSandInteraction/Textures/SSI_T_Footprint_D
  }
  n28 = ScalarParameter() {parameterName=ShoeMark_BlurMark_Flat}
  n29 = MaterialFunctionCall(Normal=@n27.RGB, Flatness=LinearInterpolate(A=LinearInterpolate(B=@n28, Alpha=@n1.BlurMarkScale) {constA=1}, Alpha=@n1.MarkFlat)) {materialFunction=&FlattenNormal}
  n30 = ScalarParameter() {parameterName=AdditionalNormal_UVScale}
  n31 = TextureSampleParameter2D(UVs=Add(
    A=MaterialFunctionCall(UVs=@n14, Texture Scale=@n30) {materialFunction=&ScaleUVsByCenter}.UVs
    B=@n11.Random
   )) {
   parameterName=AdditionalNormal
   samplerSource=SSM_Wrap_WorldGroupSettings
   samplerType=SAMPLERTYPE_Normal
   texture=/Game/Global/Textures/Default/Tileable_Patterns/SnowFlakeTest_normal
  }
  n32 = ScalarParameter() {parameterName=AdditionalNormal_Flat}
  n33 = MaterialFunctionCall(UVs=@n14, Texture Scale=ScalarParameter() {defaultValue=1, parameterName=AdditionalNormalMaskScale}) {materialFunction=&ScaleUVsByCenter}
  n34 = ScalarParameter() {parameterName=UseAdditionalNormal}
  n35 = Multiply(B=TextureSampleParameter2D(UVs=LinearInterpolate(A=@n15.UVs, B=@n33.UVs, Alpha=@n34)) {parameterName=MaskTexture, texture=&SSI_T_FootprintMasks}.B) {constA=1.7, constB=2}
  n36 = LinearInterpolate(
   A=MaterialFunctionCall(Normal=@n31.RGB, Flatness=LinearInterpolate(A=@n32, Alpha=@n1.MarkFlat) {constA=1}) {materialFunction=&FlattenNormal}.Result
   B=Constant3Vector() {constant=<rgba 0 0 1 0>}
   Alpha=OneMinus(None=Saturate(None=@n35))
  )
  Normalize(VectorInput=LinearInterpolate(A=@n29.Result, B=@n36))
  n37 = Constant3Vector() {constant=<rgba 0 0 1 0>}
  n38 = MaterialFunctionCall(Float3=NamedRerouteUsage()) {materialFunction=&BreakOutFloat3Components}
  n39 = MaterialFunctionCall(BaseNormal=@n36, AdditionalNormal=@n29.Result) {materialFunction=&BlendAngleCorrectedNormals}
  n40 = ScalarParameter() {parameterName=UseAdditionalNormal}
  n41 = TextureSampleParameter2D(UVs=@n14) {
   parameterName=ShoeMark_SharpMark
   samplerSource=SSM_Clamp_WorldGroupSettings
   samplerType=SAMPLERTYPE_Normal
   texture=/Game/Effects/SnowSandInteraction/Textures/SSI_T_Footprint_A
  }
  n42 = LinearInterpolate(A=ScalarParameter() {parameterName=ShoeMark_SharpMark_Flat}, Alpha=@n1.MarkFlat)
  n43 = Multiply(
   A=AppendVector(A=@n12, B=Constant() {r=1})
   B=MaterialFunctionCall(
    BaseNormal=LinearInterpolate(A=@n29.Result, B=@n39.Result, Alpha=@n40)
    AdditionalNormal=MaterialFunctionCall(Normal=@n41.RGB, Flatness=@n42) {materialFunction=&FlattenNormal}.Result
   ) {materialFunction=&BlendAngleCorrectedNormals}.Result
  )
  n44 = MaterialFunctionCall(
   BaseNormal=MaterialFunctionCall(Vector=NamedRerouteUsage(), Quat=@n7.RGBA, Invert Quaternion=@n8) {materialFunction=&MultiplyVectorWithQuaternion}.Result
   AdditionalNormal=@n43
  ) {materialFunction=&BlendAngleCorrectedNormals}
  n45 = If(A=Abs(None=@n38.G), "A > B"=@n43, "A == B"=@n44.Result, "A < B"=@n44.Result) {constB=0.6}
  n46 = If(A=Abs(None=@n38.R), "A > B"=@n43, "A == B"=@n45, "A < B"=@n45) {constB=0.6}
  ShadingPathSwitch(Default=@n46, Deferred=@n46, Forward=@n37, Mobile=@n37)
 ]
}
